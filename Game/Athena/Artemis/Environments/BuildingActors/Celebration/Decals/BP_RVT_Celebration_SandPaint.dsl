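; Placeable static mesh decal actor for sand paint in the Celebration environment set, apparently rendered through Runtime Virtual Texture going by its naming.
; The construction script creates a dynamic instance of the Celebration sand decal material on slot 0 and applies the Opacity, Erosion and Diffuse variables.
; Decal Tint is converted to a vector and set as DecalColor across the component's materials.
; BeginPlay, overlap and Tick events are empty.
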(event EventBeginPlay)

(event Collision|EventActorBeginOverlap
  (params (OtherActor :type Actor%20Object%20Reference)))

(event EventTick
  (params (DeltaSeconds :type Float%20%28single-precision%29)))

(fn ConstructionScript
  (bind _staticmeshcomponent (Variables|StaticMeshActor|GetStaticMeshComponent))
  (bind _returnvalue (Rendering|Material|CreateDynamicMaterialInstance _staticmeshcomponent 0 "/Game/Environments/Artemis/Sets/Celebration/Materials/Master_Materials/M_RVT_Decal_Celebration_Sand.M_RVT_Decal_Celebration_Sand" "None" :!class /Script/Engine.PrimitiveComponent))
  (Rendering|Material|SetScalarParameterValue _returnvalue "Opacity" (Variables|Colors|GetOpacity) :!class /Script/Engine.MaterialInstanceDynamic)
  (Rendering|Material|SetScalarParameterValue _returnvalue "Erosion" (Variables|Colors|GetErosion) :!class /Script/Engine.MaterialInstanceDynamic)
  (Rendering|Material|SetTextureParameterValue _returnvalue "Diffuse" (Variables|Colors|GetDiffuse))
  (Rendering|Material|SetVectorParameterValueonMaterials _staticmeshcomponent "DecalColor" (Math|Conversions|ToVector(LinearColor) (Variables|Colors|GetDecal%20Tint))))
